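# Landmass brush material function that outputs base height, previous landscape height, and linear and S-curve edge falloff for landscape brushes.
# UseShapeMask blends between procedural edge falloff and a ShapeMask texture remapped by falloff terms, with octave noise and overridable width and offset inputs.
# Base height adds brush tilt to BrushZ, while previous height comes from HeightRT sampled at texel centers and unpacked from two RG channels.
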MaterialFunction /Landmass/Landscape/BlueprintBrushes/Materials/BrushMaterials/BasicFunctions/LandmassBrush_ShapeMask
graph {nodes [
  n1 = LinearInterpolate(A=NamedRerouteUsage(), B=NamedRerouteUsage(), Alpha=ScalarParameter() {parameterName=UseShapeMask})
  FunctionOutput(None=@n1) {outputName=Brush Edge Falloff Linear, sortPriority=30}
  FunctionOutput(None=Add(
    A=MaterialFunctionCall!(Enable) {materialFunction=/Landmass/Landscape/BlueprintBrushes/MF/Landmass_BrushTilt}.Tilt
    B=ScalarParameter() {parameterName=BrushZ}
   )) {outputName=Brush Base Height}
  n2 = TextureSampleParameter2D(UVs=MaterialFunctionCall() {materialFunction=/Landmass/Landscape/BlueprintBrushes/MF/PointSampleLandscapeUVs}.TexelCenterUVs) {parameterName=HeightRT, texture=/Engine/EngineResources/Black_Low.Black}
  FunctionOutput(None=MaterialFunctionCall!(RG=ComponentMask(None=@n2.RGB) {g=true, r=true}) {materialFunction=/Landmass/Landscape/OldPrototype_BP/Mats/MF/UnpackRG8Height}.Result) {outputName=Previous Height, sortPriority=10}
  FunctionOutput(None=MaterialFunctionCall(In=@n1, Power=ScalarParameter() {defaultValue=1, parameterName=FalloffPower}) {materialFunction=&SCurve}.Result) {outputName=Brush Edge Falloff, sortPriority=20}
  n3 = Subtract(A=TextureSampleParameter2D(UVs=MaterialFunctionCall() {materialFunction=/Landmass/Landscape/BlueprintBrushes/MF/LandmassBrush_UVs}.UVs) {parameterName=ShapeMask, texture=/Engine/ArtTools/RenderToTexture/Textures/T_EV_BlankWhite_01}.R) {constB=0.5}
  n4 = Divide(A=Add(A=NamedRerouteUsage(), B=NamedRerouteUsage()))
  n5 = Divide(A=Divide(A=Subtract(A=@n3, B=@n4), B=NamedRerouteUsage())) {constB=0.5}
  NamedRerouteDeclaration(None=@n5) {name=ShapeMask, nodeColor=<rgba 1 0 0.5859 1>}
  NamedRerouteDeclaration(None=FunctionInput(Preview=ScalarParameter() {defaultValue=0.1, parameterName=FalloffWidth}) {bUsePreviewValueAsDefault=true, inputName=Falloff Width Override, inputType=FunctionInput_Scalar}) {name=Falloff Width, nodeColor=<rgba 0.0547 0 1 1>}
  NamedRerouteDeclaration(None=FunctionInput(Preview=ScalarParameter() {parameterName=FalloffOffset}) {bUsePreviewValueAsDefault=true, inputName=Falloff Offset Override, inputType=FunctionInput_Scalar, sortPriority=1}) {name=Falloff Offset, nodeColor=<rgba 0 1 0.9297 1>}
  n6 = MaterialFunctionCall!(
   UVs
   Falloff Width=NamedRerouteUsage()
   Falloff Radius=ScalarParameter() {defaultValue=0.2, parameterName=FalloffRadius}
   Offset=Add(A=NamedRerouteUsage(), B=NamedRerouteUsage())
  ) {materialFunction=/Landmass/Landscape/BlueprintBrushes/MF/LandmassBrush_EdgeFalloff}
  NamedRerouteDeclaration(None=@n6.Edge Mask) {name=Default Edge Falloff, nodeColor=<rgba 0.8281 0 1 1>}
  n7 = AppendVector(
   A=Multiply(
    A=MaterialFunctionCall() {materialFunction=&LandmassBrush_UVs}.UVs
    B=ScalarParameter() {defaultValue=2, parameterName=EdgeNoiseTiling}
   )
   B=ScalarParameter() {defaultValue=1, parameterName=EdgeNoiseSeed}
  )
  n8 = ScalarParameter() {defaultValue=4, parameterName=EdgeNoiseOctaves}
  n9 = ScalarParameter() {defaultValue=0.35, parameterName=EdgeNoisePersistence}
  NamedRerouteDeclaration(None=Multiply(
    A=MaterialFunctionCall!(Position=@n7, Octaves=@n8, Persistence=@n9, Turbulence) {materialFunction=/Landmass/Landscape/BlueprintBrushes/MF/LandmassBrush_Noise}.Result
    B=ScalarParameter() {defaultValue=0.2, parameterName=EdgeNoiseStrength}
   )) {name=Edge Falloff Noise, nodeColor=<rgba 0.0786 0.3441 1 1>}
  Saturate(None=@n5)
  Saturate(None=@n6.Edge Mask)
 ]}
