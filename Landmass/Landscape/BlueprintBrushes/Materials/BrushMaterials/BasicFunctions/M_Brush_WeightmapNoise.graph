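# Material function for Landmass landscape brushes that breaks up an incoming brush alpha with procedural noise for weightmap painting.
# LandmassBrush_Noise samples tiled, seeded brush UVs, then invert, floor remap, contrast and multiply parameters shape the result.
# Named reroutes carry two noise variants, apparently feeding a HeightLerp-modulated alpha edge and a plain alpha-times-noise product.
# The MultiplyNoise parameter blends between the two modes for the saturated Weightmap Noise output.
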MaterialFunction /Landmass/Landscape/BlueprintBrushes/Materials/BrushMaterials/BasicFunctions/M_Brush_WeightmapNoise
graph {nodes [
  n1 = NamedRerouteUsage()
  n2 = NamedRerouteUsage()
  n3 = ScalarParameter() {parameterName=MultiplyNoise}
  FunctionOutput(None=Saturate(None=LinearInterpolate(A=@n1, B=@n2, Alpha=@n3))) {outputName=Weightmap Noise}
  NamedRerouteDeclaration(None=FunctionInput() {inputName=Brush Alpha, inputType=FunctionInput_Scalar}) {name=Alpha In, nodeColor=<rgba 0 1 0.1797 1>}
  n4 = Clamp(None=NamedRerouteUsage()) {maxDefault=0.999, minDefault=0.001}
  n5 = ScalarParameter() {defaultValue=1, parameterName=WeightNoiseContrast}
  NamedRerouteDeclaration(None=MaterialFunctionCall(Transition Phase=NamedRerouteUsage(), Height Texture=LinearInterpolate(B=@n4, Alpha=@n5) {constA=0.5}, Contrast=@n5) {materialFunction=&HeightLerp}.Alpha) {name=Alpha Modulated Edge, nodeColor=<rgba 0 1 0.7891 1>}
  NamedRerouteDeclaration(None=Multiply(A=NamedRerouteUsage(), B=NamedRerouteUsage())) {name=Alpha Multiplied by Noise, nodeColor=<rgba 0 1 0.0391 1>}
  n6 = ScalarParameter() {defaultValue=1, parameterName=WeightNoiseMultiply}
  n7 = MaterialFunctionCall() {materialFunction=/Landmass/Landscape/BlueprintBrushes/MF/LandmassBrush_UVs}
  n8 = ScalarParameter() {defaultValue=8, parameterName=WeightNoiseTiling}
  n9 = MaterialFunctionCall!(
   Position=AppendVector(A=Multiply(A=@n7.UVs, B=@n8), B=ScalarParameter() {defaultValue=1, parameterName=WeightNoiseSeed})
   Octaves=ScalarParameter() {defaultValue=4, parameterName=WeightNoiseOctaves}
   Persistence=ScalarParameter() {defaultValue=0.35, parameterName=WeightNoisePersistence}
   Turbulence=ScalarParameter() {parameterName=WeightNoiseTurbulence}
  ) {materialFunction=/Landmass/Landscape/BlueprintBrushes/MF/LandmassBrush_Noise}
  n10 = LinearInterpolate(A=@n9.Result, B=OneMinus(None=@n9.Result), Alpha=ScalarParameter() {parameterName=WeightNoiseInvert})
  n11 = ScalarParameter() {parameterName=WeightNoiseFloor}
  n12 = Max(A=Divide(A=Subtract(A=@n10, B=@n11), B=OneMinus(None=@n11))) {constB=0}
  NamedRerouteDeclaration(None=Saturate(None=Multiply(A=@n12, B=@n6))) {name=Weightmap Noise Heightlerp, nodeColor=<rgba 0.1719 0 1 1>}
  n13 = Power(Base=@n12, Exp=ScalarParameter() {defaultValue=1, parameterName=WeightNoiseContrast})
  n14 = ScalarParameter() {defaultValue=1, parameterName=WeightNoiseMultiply}
  NamedRerouteDeclaration(None=Saturate(None=Multiply(A=@n13, B=@n14))) {name=Weightmap Noise, nodeColor=<rgba 0.1719 0 1 1>}
 ]}
